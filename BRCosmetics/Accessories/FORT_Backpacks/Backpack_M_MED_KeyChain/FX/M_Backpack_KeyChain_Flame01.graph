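# Translucent unlit flame trail material for the KeyChain backpack, enabled for Niagara ribbons and mesh particles.
# World position offset combines SplineThicken width, sine and noise waving, and an optional bend driven by dynamic parameters.
# Opacity multiplies panning trail and noise textures with head/tail masks, particle alpha, optional dissolve and appear, angle fades and depth fade.
# Emissive blends scrolling caustics, curve-atlas gradient and added color, tinted by particle color through MF_FX_EmissiveAdjustments.
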Material /BRCosmetics/Accessories/FORT_Backpacks/Backpack_M_MED_KeyChain/FX/M_Backpack_KeyChain_Flame01
graph {
 properties {
  bUsedWithNanite=true
  bUsedWithNiagaraMeshParticles=true
  bUsedWithNiagaraRibbons=true
  blendMode=BLEND_Translucent
  pixelDepthOffsetMode=PDOM_Legacy
  shadingModel=MSM_Unlit
  usedShadingModels=Unlit
 }
 nodes [
  n1 = DynamicParameter() {defaultValue=<rgba 5 15 1 1>, paramNames [WidthBase,WidthTip,UV_AddU,UV_AddV]}
  n2 = VertexColor()
  n3 = ScalarParameter() {parameterName=Use_VertexColor}
  n4 = MaterialFunctionCall(
   WidthBase=LinearInterpolate(A=@n1.WidthBase, B=Multiply(A=@n2.R, B=@n1.WidthBase), Alpha=@n3)
   WidthTip=LinearInterpolate(A=@n1.WidthTip, B=Multiply(A=@n2.R, B=@n1.WidthTip), Alpha=@n3)
   Expand U or V UV Channel=StaticBoolParameter() {parameterName="Expand U or V UV Channel (SB)"}
  ) {materialFunction=&SplineThicken}
  n5 = Multiply(A=TextureCoordinate(), B=ComponentMask(None=VectorParameter() {parameterName=WPO_UScale, defaultValue=<rgba 1 1 0 0>}.RGB) {g=true, r=true})
  n6 = LinearInterpolate(A=@n5, B=Multiply(A=@n5, B=VertexColor().R), Alpha=ScalarParameter() {parameterName=Use_VertexColor})
  n7 = Multiply(A=Time(), B=ScalarParameter() {defaultValue=1, parameterName=TimeMult})
  n8 = Sine(None=Add(A=ComponentMask(None=@n6) {r=true}, B=Multiply(A=@n7) {constB=-0.5})) {period=0.4}
  n9 = TextureCoordinate() {uTiling=3, vTiling=0}
  n10 = DynamicParameter() {defaultValue=<rgba 5 15 1 1>, paramNames [WidthBase,WidthTip,UV_AddU,UV_AddV]}
  n11 = AppendVector(A=@n10.UV_AddU, B=@n10.UV_AddU)
  n12 = Time()
  n13 = ScalarParameter() {defaultValue=1, parameterName=TimeMult}
  n14 = TextureSample(UVs=Panner(Coordinate=Add(A=@n9, B=@n11), Time=Multiply(A=@n12, B=@n13)) {speedX=-0.3}) {samplerType=SAMPLERTYPE_LinearGrayscale, texture=/Engine/Functions/Engine_MaterialFunctions02/ExampleContent/Textures/LowResBlurredNoise}
  n15 = Multiply(
   A=VectorParameter() {parameterName=WPODirection, defaultValue=<rgba 0 1 0 1>}.RGB
   B=LinearInterpolate(Alpha=LinearInterpolate(A=@n8, B=@n14.R)) {constA=-1}
  )
  n16 = ScalarParameter() {defaultValue=0.5, parameterName=WPOMult}
  n17 = LinearInterpolate(
   B=DynamicParameter() {parameterIndex=1, paramNames [WPOScale, Use Custom UScale, DissolveRange, Appear]}.WPOScale
   Alpha=ScalarParameter() {parameterName=UseCustom WPO}
  ) {constA=1}
  n18 = Time()
  n19 = ScalarParameter() {defaultValue=1, parameterName=TimeMult}
  n20 = Add(A=ComponentMask(None=TextureCoordinate()) {r=true}, B=Multiply(A=Multiply(A=@n18, B=@n19)) {constB=-0.5})
  n21 = Multiply(
   A=VectorParameter() {parameterName=TrailWPODirection, defaultValue=<rgba 0 1 0 1>}.RGB
   B=LinearInterpolate(Alpha=Sine(None=@n20) {period=0.4})
  )
  n22 = LinearInterpolate(
   A=Multiply(A=Transform(None=@n15), B=Multiply(A=@n16, B=@n17)) {constB=0.5}
   B=Multiply(
    A=Transform(None=@n21) {transformSourceType=TRANSFORMSOURCE_Local, transformType=TRANSFORM_Instance}
    B=DynamicParameter() {parameterIndex=1, paramNames [WPOScale, Use Custom UScale, DissolveRange, Appear]}.WPOScale
   )
   Alpha=ScalarParameter() {parameterName=UseTrailWpo}
  )
  n23 = DynamicParameter() {parameterIndex=2, defaultValue=<rgba 5 15 1 1>, paramNames [X,Y,Z,WPOIntensity]}
  n24 = ScalarParameter() {defaultValue=50, parameterName=InterpolateWPORange}
  n25 = Clamp(
   None=Subtract(
    A=MaterialFunctionCall(X=@n23.X, Y=@n23.X, Z=@n23.X) {materialFunction=&MakeFloat3}.Result
    B=ParticlePositionWS()
   )
   Min=Multiply(A=@n24) {constB=-1}
   Max=@n24
  )
  n26 = TextureCoordinate()
  n27 = Multiply(A=@n23.WPOIntensity, B=Power(Base=ComponentMask(None=@n26) {r=true}, Exp=ScalarParameter() {defaultValue=3, parameterName=WPOBendMask}))
  Add(
   A=LinearInterpolate(B=Add(A=@n4.WorldPositionOffset, B=@n22), Alpha=ScalarParameter() {defaultValue=1, parameterName=UseSplineThicken})
   B=LinearInterpolate(B=LinearInterpolate(B=@n25, Alpha=@n27), Alpha=ScalarParameter() {parameterName=UseInterpolateWPO})
  )
  n28 = Panner(Coordinate=TextureCoordinate() {uTiling=2, vTiling=0.5}) {speedX=-1.5}
  n29 = Multiply(A=MaterialFunctionCall(
    Transition Phase=DynamicParameter() {parameterIndex=1, paramNames [WPOScale, Use Custom UScale, DissolveRange, Appear]}.DissolveRange
    Height Texture=TextureSample(UVs=@n28) {texture=/Game/Effects/Niagara/MasterMaterials/SplineThickenTrail/Texture/T_Pickaxe_SoapPocket_NoiseGlow}.R
    Contrast=Constant() {r=1}
   ) {materialFunction=&HeightLerp}.Alpha)
  n30 = ScalarParameter() {parameterName=UseGlobleDissolve}
  n31 = ParticleColor()
  n32 = DynamicParameter() {defaultValue=<rgba 5 15 1 1>, paramNames [WidthBase,WidthTip,UV_AddU,UV_AddV]}
  n33 = Multiply(
   A=Add(A=TextureCoordinate() {uTiling=2}, B=AppendVector(A=@n32.UV_AddU, B=@n32.UV_AddU))
   B=AppendVector(A=ScalarParameter() {defaultValue=1, parameterName=UScale}, B=ScalarParameter() {defaultValue=1, parameterName=VScale})
  )
  n34 = ComponentMask(None=VectorParameter() {parameterName=UVscale, defaultValue=<rgba 1 1 0 0>}.RGB) {g=true, r=true}
  n35 = Time()
  n36 = ScalarParameter() {defaultValue=1, parameterName=TimeMult}
  n37 = TextureSample(UVs=Panner(Coordinate=Multiply(A=@n33, B=@n34), Time=Multiply(A=@n35, B=@n36)) {speedX=-0.7}) {samplerType=SAMPLERTYPE_LinearGrayscale, texture=/Game/Global/Textures/Default/Caustics/T_Nosie_Caustics_Grayscale}
  n38 = ComponentMask(None=TextureCoordinate()) {r=true}
  n39 = Saturate(None=LinearInterpolate(
    A=ScalarParameter() {parameterName=Dissolve mask Min}
    B=ScalarParameter() {defaultValue=0.5, parameterName=Dissolve mask Max}
    Alpha=Power(Base=@n38) {constExponent=1.3}
   ) {constA=0.1, constB=2})
  n40 = Saturate(None=LinearInterpolate(
    A=ScalarParameter() {defaultValue=-2, parameterName=Disappear min}
    B=ScalarParameter() {defaultValue=2, parameterName=Disappear max}
    Alpha=Power(Base=@n37.R, Exp=@n39) {constExponent=0.2}
   ) {constA=-2, constB=2})
  n41 = Add(A=TextureCoordinate() {uTiling=2, vTiling=0.9}, B=Constant2Vector() {g=0.05})
  n42 = DynamicParameter() {defaultValue=<rgba 5 15 1 1>, paramNames [WidthBase,WidthTip,UV_AddU,UV_AddV]}
  n43 = AppendVector(A=@n42.UV_AddU, B=@n42.UV_AddU)
  n44 = ScalarParameter() {defaultValue=1, parameterName=UScale}
  n45 = ScalarParameter() {defaultValue=1, parameterName=VScale}
  n46 = Multiply(
   A=Multiply(A=Add(A=@n41, B=@n43), B=AppendVector(A=@n44, B=@n45))
   B=ComponentMask(None=VectorParameter() {parameterName=UVscale, defaultValue=<rgba 1 1 0 0>}.RGB) {g=true, r=true}
  )
  n47 = Multiply(A=Time(), B=ScalarParameter() {defaultValue=1, parameterName=TimeMult})
  n48 = Multiply(A=TextureSample(UVs=Panner(Coordinate=@n46, Time=@n47) {speedX=-0.5}) {samplerType=SAMPLERTYPE_Grayscale, texture=&T_Smoke_Trail_Soft}.R) {constB=1.5}
  n49 = TextureCoordinate()
  n50 = DynamicParameter() {defaultValue=<rgba 5 15 1 1>, paramNames [WidthBase,WidthTip,UV_AddU,UV_AddV]}
  n51 = AppendVector(A=@n50.UV_AddU, B=@n50.UV_AddU)
  n52 = ScalarParameter() {defaultValue=1, parameterName=UScale}
  n53 = ScalarParameter() {defaultValue=1, parameterName=VScale}
  n54 = Multiply(
   A=Multiply(A=Add(A=@n49, B=@n51), B=AppendVector(A=@n52, B=@n53))
   B=ComponentMask(None=VectorParameter() {parameterName=UVscale, defaultValue=<rgba 1 1 0 0>}.RGB) {g=true, r=true}
  )
  n55 = ScalarParameter() {defaultValue=1, parameterName=TimeMult}
  n56 = VertexColor()
  n57 = Multiply(A=Time(), B=LinearInterpolate(A=@n55, B=Multiply(A=@n55, B=@n56.R), Alpha=ScalarParameter() {parameterName=Use_VertexColor}))
  n58 = Multiply(A=TextureSampleParameter2D(UVs=Panner(Coordinate=@n54, Time=@n57) {speedX=-0.5}) {parameterName=TrailTexture, texture=/Game/Accessories/FORT_Backpacks/Backpack_M_MED_KeyChain/FX/T_Backpack_KeyChain_Flame}.R)
  n59 = ScalarParameter() {defaultValue=0.3, parameterName=TrailTexture_LerpAlpha}
  n60 = DynamicParameter() {defaultValue=<rgba 5 15 1 1>, paramNames [WidthBase,WidthTip,UV_AddU,UV_AddV]}
  n61 = Add(A=TextureCoordinate() {uTiling=2}, B=AppendVector(A=@n60.UV_AddU, B=@n60.UV_AddU))
  n62 = AppendVector(A=ScalarParameter() {defaultValue=1, parameterName=UScale}, B=ScalarParameter() {defaultValue=1, parameterName=VScale})
  n63 = VectorParameter() {parameterName=UVscale, defaultValue=<rgba 1 1 0 0>}
  n64 = TextureSampleParameter2D(UVs=Multiply(A=Multiply(A=@n61, B=@n62), B=ComponentMask(None=@n63.RGB) {g=true, r=true})) {
   parameterName=LerpTexture
   samplerType=SAMPLERTYPE_LinearGrayscale
   texture=&LowResBlurredNoise
  }
  n65 = Multiply(
   A=Saturate(None=LinearInterpolate(A=@n48, B=@n58, Alpha=@n59) {constAlpha=0.3})
   B=Power(Base=@n64.R, Exp=ScalarParameter() {defaultValue=1, parameterName=Mask power}) {constExponent=1}
  )
  n66 = OneMinus(None=ComponentMask(None=TextureCoordinate()) {r=true})
  n67 = ScalarParameter() {defaultValue=2, parameterName=Tail Power}
  n68 = Saturate(None=Multiply(A=Power(Base=@n66, Exp=@n67) {constExponent=4}, B=ScalarParameter() {defaultValue=8, parameterName=Tail edge hardness}) {constB=5})
  n69 = ComponentMask(None=TextureCoordinate()) {r=true}
  n70 = ScalarParameter() {defaultValue=1.5, parameterName=LeadMaskPower}
  n71 = Saturate(None=Multiply(A=Power(Base=@n69, Exp=@n70) {constExponent=1.5}, B=ScalarParameter() {defaultValue=10, parameterName=LeadMaskHardness}) {constB=10})
  n72 = Multiply(A=Multiply(A=Multiply(A=@n65, B=@n40), B=Multiply(A=@n68, B=@n71))) {constB=2}
  n73 = Multiply(
   A=@n72
   B=MaterialFunctionCall("FadeSensitivity (optional)"=ScalarParameter() {parameterName=AngleFadeSensitivity}) {materialFunction=&FadeBasedOnViewAngle}.Result
  )
  n74 = ScalarParameter() {parameterName=UseAngleFade}
  n75 = OneMinus(None=ComponentMask(None=TextureCoordinate()) {r=true})
  n76 = LinearInterpolate(Alpha=DynamicParameter() {parameterIndex=1, paramNames [WPOScale, Use Custom UScale, DissolveRange, Appear]}.Appear) {constA=-1}
  n77 = Saturate(None=Multiply(A=Add(A=@n75, B=@n76)) {constB=3})
  n78 = ScalarParameter() {parameterName=UseAppear}
  n79 = Multiply(A=@n31.A, B=Multiply(A=LinearInterpolate(A=@n72, B=@n73, Alpha=@n74), B=LinearInterpolate(B=@n77, Alpha=@n78) {constA=1}))
  n80 = MaterialFunctionCall("FadeSensitivity (optional)"=ScalarParameter() {parameterName=MeshFade}) {materialFunction=&FadeBasedOnViewAngle}
  DepthFade(
   Opacity=Multiply(A=Multiply(A=@n79, B=@n80.Result), B=LinearInterpolate(B=@n29, Alpha=@n30) {constA=1})
   FadeDistance=ScalarParameter() {defaultValue=5, parameterName=DepthFadeDistance}
  ) {fadeDistanceDefault=5}
  n81 = DynamicParameter() {defaultValue=<rgba 5 15 1 1>, paramNames [WidthBase,WidthTip,UV_AddU,UV_AddV]}
  n82 = Multiply(
   A=Add(A=TextureCoordinate() {uTiling=1.5, vTiling=0.5}, B=AppendVector(A=@n81.UV_AddU, B=@n81.UV_AddU))
   B=AppendVector(A=ScalarParameter() {defaultValue=1, parameterName=UScale}, B=ScalarParameter() {defaultValue=1, parameterName=VScale})
  )
  n83 = ComponentMask(None=VectorParameter() {parameterName=UVscale, defaultValue=<rgba 1 1 0 0>}.RGB) {g=true, r=true}
  n84 = Time()
  n85 = ScalarParameter() {defaultValue=1, parameterName=TimeMult}
  n86 = TextureSampleParameter2D(UVs=Panner(Coordinate=Multiply(A=@n82, B=@n83), Time=Multiply(A=@n84, B=@n85)) {speedX=-0.5}) {parameterName=EmissiveTextureA, samplerType=SAMPLERTYPE_LinearColor, texture=/Game/Effects/Fort_Effects/BluGlow/Violentium/Violentium_Mask}
  n87 = Power(Base=@n86.R)
  n88 = Multiply(A=Multiply(A=VectorParameter() {parameterName=AddedColor, defaultValue=<rgba 1 0.2151 0 1>}.RGB, B=LinearInterpolate(Alpha=@n87)))
  n89 = TextureCoordinate()
  n90 = Power(Base=Sine(None=ComponentMask(None=@n89) {g=true}) {period=2}) {constExponent=3}
  n91 = VectorParameter() {parameterName=UVscale, defaultValue=<rgba 1 1 0 0>}
  n92 = TextureSample(UVs=Multiply(A=TextureCoordinate(), B=ComponentMask(None=@n91.RGB) {g=true, r=true})) {samplerType=SAMPLERTYPE_LinearGrayscale, texture=&LowResBlurredNoise}
  n93 = LinearInterpolate(A=Saturate(None=Multiply(A=@n90)), B=@n92.R)
  n94 = CurveAtlasRowParameter(CurveTime=@n93) {
   atlas=/Game/Accessories/FORT_Backpacks/Backpack_M_MED_KeyChain/FX/CAtlas_KeyChain_FX_Color
   curve=/Game/Accessories/FORT_Backpacks/Backpack_M_MED_KeyChain/FX/CV_KeyChain_Flame_Color1
   defaultValue=1
   parameterName=Color2
  }
  n95 = ScalarParameter() {defaultValue=1, parameterName=UseColorCurve}
  n96 = LinearInterpolate(
   B=DynamicParameter() {parameterIndex=1, paramNames [WPOScale, Use Custom UScale, DissolveRange, Appear]}.Use Custom UScale
   Alpha=ScalarParameter() {parameterName=Use Custom UScale}
  ) {constA=1}
  n97 = Constant() {r=1}
  n98 = Multiply(
   A=Multiply(A=TextureCoordinate() {uTiling=1.2, vTiling=0.6}, B=AppendVector(A=@n96, B=@n97))
   B=ComponentMask(None=VectorParameter() {parameterName=UVscale, defaultValue=<rgba 1 1 0 0>}.RGB) {g=true, r=true}
  )
  n99 = Panner(
   Coordinate=TextureCoordinate() {uTiling=0.5, vTiling=0.3}
   Time=Multiply(A=Time(), B=ScalarParameter() {defaultValue=1, parameterName=TimeMult})
  ) {speedX=-0.05}
  n100 = Multiply(A=ComponentMask(None=TextureSample(UVs=@n99) {texture=&T_Inky_Smoke_Tile}.RGB) {g=true, r=true}) {constB=0.2}
  n101 = ScalarParameter() {defaultValue=1, parameterName=UScale}
  n102 = ScalarParameter() {defaultValue=1, parameterName=VScale}
  n103 = DynamicParameter() {defaultValue=<rgba 5 15 1 1>, paramNames [WidthBase,WidthTip,UV_AddU,UV_AddV]}
  n104 = Add(A=Multiply(A=Add(A=@n98, B=@n100), B=AppendVector(A=@n101, B=@n102)), B=AppendVector(A=@n103.UV_AddU, B=@n103.UV_AddU))
  n105 = Multiply(A=Time(), B=ScalarParameter() {defaultValue=1, parameterName=TimeMult})
  n106 = Power(Base=TextureSampleParameter2D(UVs=Panner(Coordinate=@n104, Time=@n105) {speedX=-0.5}) {parameterName=EmissiveTextureB, texture=&T_ShieldCaustics}.R) {constExponent=0.5}
  n107 = Multiply(
   A=LinearInterpolate(Alpha=@n90) {constA=0.2}
   B=Multiply(A=LinearInterpolate(Alpha=@n106) {constA=0.3, constB=2}, B=LinearInterpolate(A=@n93, B=@n94, Alpha=@n95)) {constA=1}
  )
  MaterialFunctionCall(Emissive=Multiply(A=@n31.RGB, B=Add(A=@n88, B=@n107))) {materialFunction=&MF_FX_EmissiveAdjustments}
  Multiply(A=Time(), B=ScalarParameter() {defaultValue=1, parameterName=TimeMult})
 ]
}
