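# Animation Blueprint for the KeyChain backpack cosmetic that bobs its base_spinner bone and spins it when the player gets a kill.
# Each update computes a sine-driven Z offset and a tangent-driven pitch, applied additively through two Modify Bone nodes.
# A watch on the AthenaKills cosmetic stat sets Active_Kill for 1.5 seconds.
# That flag becomes the rotation node's alpha, which is interpolated to ease the spin in and out.
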AnimBlueprint /BRCosmetics/Accessories/FORT_Backpacks/Backpack_M_MED_KeyChain/Meshes/M_MED_KeyChain_Pack_AnimBP
target_skeleton=&/M_MED_KeyChain_Pack_Skeleton
graphs [
 AnimGraph {nodes [
   n1 = "Transform (Modify) Bone - Bone: base_spinner"(
    ComponentPose=Local To Component(LocalPose=Input Pose())
    Alpha=1
    Translation=Get float() -> (float:Vector)
    Rotation="0,0,0"
    Scale="1,1,1"
   ) {translationMode=BMM_Additive, translationSpace=BCS_BoneSpace, alphaBoolBlend {blendOption=Linear}, boneToModify {boneName=base_spinner}}
   n2 = "Transform (Modify) Bone - Bone: base_spinner"(
    ComponentPose=@n1
    Alpha=Get Kill_Happened() -> (Kill_Happened:double)
    Translation="0,0,0"
    Rotation=Get Spinning_Rotator() -> (Spinning_Rotator:Rotator)
    Scale="1,1,1"
   ) {
    rotationMode=BMM_Additive
    rotationSpace=BCS_BoneSpace
    translationMode=BMM_Additive
    translationSpace=BCS_BoneSpace
    alphaBoolBlend {blendOption=Linear}
    alphaScaleBiasClamp {bInterpResult=true, interpSpeedDecreasing=5, interpSpeedIncreasing=5}
    boneToModify {boneName=base_spinner}
    rotation {yaw=359}
   }
   Output Pose(Result=Component To Local(ComponentPose=@n2)) {layerGroup=None}
  ]}
 EventGraph {nodes [
   Event BlueprintUpdateAnimation()
   > n1 = Sequence!(execute) -> (then_0, then_1, then_2:Exec, then_3:Exec)
   n2 = "Sin (Radians)"(A=float * float(A=GetRealTimeSeconds(), B=2))
   > Set float(float:Vector=MakeVector(Y=0, Z=float / float(A=@n2, B=-2) {is=title K2Node_CallFunction})) -> (Output_Get:Vector)
   Get!(execute:Exec=@n1.then_1) -> (then:Exec, else:Exec, Skel_Mesh:Skeletal Mesh Component Object Reference)
   > Sequence!(execute) -> (then_0, then_1)
   > n3 = Branch(Condition=Get Active_Kill() -> (Active_Kill:bool))
   > "Set FInterp(Alpha, (10: 10))"("FInterp(Alpha, (10: 10))":float="Get FInterp(Alpha, (10: 10))_0"() -> ("FInterp(Alpha, (10: 10))_0":float)) -> (Output_Get:float)
   > Delay(Duration=1.5)
   > Set Active_Kill(Active_Kill:bool=false) -> (Output_Get:bool)
   "Set FInterp(Alpha, (10: 10))"(execute=@n3.else, "FInterp(Alpha, (10: 10))"=0)
   Set Kill_Happened(execute=@n1.then_2, Kill_Happened:double="To Float (Boolean)"(InBool=Get Active_Kill())) -> (Output_Get:double)
   n4 = "Tan (Degrees)"!(A:double=float * float(A=GetGameTimeInSeconds(), B=80)) -> (ReturnValue:double)
   n5 = Set Sphere_Rotation(execute=@n1.then_3, Sphere_Rotation:double=NormalizeAxis!(Angle:float=float * float(A=@n4, B=-100)) -> (ReturnValue:float)) -> (Output_Get:double)
   > Set Spinning_Rotator(Spinning_Rotator:Rotator=MakeRotator(Pitch=@n5)) -> (Output_Get:Rotator)
   Event OnCharacterCustomizationLoaded()
   > n6 = Cast To FortPlayerPawn(Object=TryGetFortPawn())
   > n7 = WatchCosmeticStat!(
    execute:Exec
    Pawn:Fort Player Pawn Object Reference=@n6
    StatTag:Fort Stat Manager Tag Structure=(TagName="StatManager.AthenaKills")
    DefaultValue:int=0
    bBroadcastInitialValue:bool=false
    InitialValueBroadcastDelay:float=-1
    ChangeValueBroadcastDelay:float=-1
    bConsumeWaitingNotifications:bool=true
   ) -> (then:Exec, OnStatChanged:Exec, StatName:Name, StatValue:int) {is=title K2Node_AsyncAction}
   Set KillCount(execute=@n7.OnStatChanged, KillCount:double="To Float (Integer)"(InInt=@n7.StatValue)) -> (Output_Get:double)
   > Set Active_Kill(Active_Kill=true)
   > Delay(Duration=1.5)
   > Set Active_Kill(Active_Kill=false)
   TryGetPawnOwner()
  ]}
]
variables [
 Active_Kill {type=boolean}
 "FInterp(Alpha, (10: 10))" {type=number}
 "FInterp(Alpha, (10: 10))_0" {type=number}
 float {type=Vector}
 Kill_Happened {type=number}
 KillCount {type=number}
 Skel_Mesh {type=SkeletalMeshComponent}
 Sphere_Rotation {type=number}
 Spinning_Rotator {type=Rotator}
 Spinning_Rotator_Faster {type=Rotator}
]
